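; BPi_TextStyleOverrides is a Blueprint interface in the Creative property editor's style editor for supplying text style overrides.
; It declares two functions, GetDefaultStyleOverride and GetTextStyleOverride, each returning a boolean alongside style values.
; Both bodies hold only default return values, so any real logic lives in implementers outside this file.
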(fn GetDefaultStyleOverride
  (return false 0 0.0 0 0))

(fn GetTextStyleOverride
  (return false 0))
