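# Creative property editor dialog panel that displays rich text in a scrolling body, with an empty-text prompt and a gamepad Edit hint.
# Update Text substitutes the local player name, toggles the prompt and picks a style set.
# With the style editor enabled, styles come from the device's BPi_TextStyleOverrides, the minigame settings machine or a default table, and device default-style overrides apply; otherwise a disabled set is used.
# UpdateColor switches dark/light backgrounds; the edit-color hint is collapsed.
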WidgetBlueprint /FortniteCoreUI/InGame/Creative/PropertyEditor/StyleEditor/CreativeEditDialogWidget_RichTextEditor
animations [Anim_IntroOutro {
  binding_count=1
  length_seconds=0.25
  object=&self:Anim_IntroOutro
  sequence {bindings [Overlay_0 {
     id=#1
     possessed_class=Overlay
     tracks [
      {
       type=MovieSceneFloatTrack
       display_name=Render Opacity
       sections [{type=MovieSceneFloatSection, end_frame=5, start_frame=0, channels [None {type=MovieSceneScriptingFloatChannel, keys [[0,0],[3,0],[5,1]]}]}]
      }
      {
       type=MovieScene2DTransformTrack
       sections [{
         type=MovieScene2DTransformSection
         end_frame=5
         start_frame=0
         channels [
          Translation.X {type=MovieSceneScriptingFloatChannel, keys [[0,-440],[5,0]]}
          Translation.Y {type=MovieSceneScriptingFloatChannel, const [0,1]}
         ]
        }]
      }
     ]
    }]}
 }]
graphs [
 EventGraph {nodes [
   n1 = Update Text!() -> (OutputDelegate, then, InText:"Text (by ref)") {is=title K2Node_CustomEvent}
   n2 = GetOwningPlayer()
   > SetText(
    self=Get DialogBody() -> (DialogBody:Common Rich Text Block Object Reference)
    InText=Format Text!(
     Format=@n1.InText
     playername:Text=Get Local Player Name from Controller!(Controller:Controller Object Reference=@n2) -> (PlayerName:Text) {is=title K2Node_CallFunction}
    ) -> (Result)
   )
   > UpdateEmptyPrompt()
   > SetRichTextStyleSet()
   n3 = Event PreConstruct()
   > Branch(Condition=@n3.IsDesignTime)
   > UpdateColor!(execute, self, InUseLightColor:bool=false) -> (then)
   Event OnInitialized()
   > SetInputAction!(
    execute
    self:Common Action Widget Object Reference=Get EditTextActionWidget() -> (EditTextActionWidget:Common Action Widget Object Reference)
    InputActionRow:Data Table Row Handle Structure=Get EditTextInputAction() -> (EditTextInputAction:Data Table Row Handle Structure)
   ) -> (then)
   > SetInputAction(
    self=Get EditColorActionWidget() -> (EditColorActionWidget:Common Action Widget Object Reference)
    InputActionRow=Get EditColorInputAction() -> (EditColorInputAction:Data Table Row Handle Structure)
   )
   > UpdateEmptyPrompt()
  ]}
 Get Device Rich Text Overrides {nodes [
   Get Device Rich Text Overrides() {is=title K2Node_FunctionEntry}
   n1 = GetOwningEditWidget!(self:Fort Creative Property Editor Dialog Base Object Reference) -> (ReturnValue:Creative Property Edit Widget User Option Base Object Reference)
   > n2 = Is Valid(InputObject=@n1)
   Return Node!(
    execute=@n2.Is Not Valid
    Success:bool=false
    RichTextStyle:Data Table Object Reference
    OverrideStyle:bool=false
    Interface:BPi Text Style Overrides Interface
   )
   > n3 = Cast To BPi_TextStyleOverrides!(
    execute
    Object=GetInitializingObject!(self:Creative Property Edit Widget User Option Base Object Reference=@n1) -> (ReturnValue:Object Reference)
   ) -> (then, CastFailed, AsBPi Text Style Overrides:BPi Text Style Overrides Interface) {is=title K2Node_DynamicCast}
   > n4 = GetTextStyleOverride!(execute, self:BPi Text Style Overrides Interface=@n3) -> (then, HasOverride:bool, RichTextStyle:Data Table Object Reference)
   > Return Node(Success=true, RichTextStyle=@n4.RichTextStyle, OverrideStyle=@n4, Interface=@n3)
  ]}
 GetCurrentBackgroundColor {nodes [
   GetCurrentBackgroundColor() {is=title K2Node_FunctionEntry}
   > Return Node!(execute, OutIsUsingLightColor:bool=Get bUseLightColorBackground() -> (bUseLightColorBackground:bool))
  ]}
 SetRichTextStyleSet {nodes [
   SetRichTextStyleSet() {is=title K2Node_FunctionEntry}
   > n1 = Branch(Condition=IsStyleEditorEnabled!() -> (ReturnValue:bool))
   > n2 = Get Device Rich Text Overrides!(execute, self) -> (then, Success:bool, RichTextStyle:Data Table Object Reference, OverrideStyle:bool, Interface:BPi Text Style Overrides Interface) {is=title K2Node_CallFunction}
   > n3 = Branch(Condition=@n2)
   > n4 = Sequence()
   > n5 = Branch(Condition=@n2.OverrideStyle)
   > SetTextStyleSet!(execute, self:Rich Text Block Object Reference=Get DialogBody(), NewTextStyleSet:Data Table Object Reference=@n2.RichTextStyle) -> (then)
   n6 = GetMinigameSettingsMachine!(ContextActor:Actor Object Reference=GetOwningPlayer()) -> (ReturnValue:Fort Minigame Settings Building Object Reference)
   n7 = Is Valid(exec=[@n5.else, @n3.else], InputObject=@n6)
   n8 = GetSelectedTextStyleSet!(self:Fort Minigame Settings Building Object Reference=@n6) -> (ReturnValue:Data Table Object Reference)
   > n9 = Is Valid(InputObject=@n8)
   n10 = Get DialogBody()
   > SetTextStyleSet(self=@n10, NewTextStyleSet=@n8)
   SetTextStyleSet(
    execute=[@n9.Is Not Valid, @n7.Is Not Valid]
    self=@n10
    NewTextStyleSet=/Game/Athena/UI/InGame/Creative/PropertyEditor/StyleEditor/RichTextStyleSets/CreativeDefaultRichTextStyles
   )
   Update Default Rich Text Style!(execute=@n4.then_1, self, Device Interface:BPi Text Style Overrides Interface=@n2.Interface) -> (then)
   SetTextStyleSet(
    execute=@n1.else
    self=@n10
    NewTextStyleSet=/FortniteCoreUI/InGame/Creative/PropertyEditor/StyleEditor/RichTextStyleSets/CreativeRichTextStyles_Disabled
   )
  ]}
 Update Default Rich Text Style {nodes [
   n1 = Update Default Rich Text Style!() -> (then, Device Interface:BPi Text Style Overrides Interface) {is=title K2Node_FunctionEntry}
   > n2 = GetDefaultStyleOverride!(execute, self:BPi Text Style Overrides Interface=@n1) -> (then, HasOverride:bool, TextColor:Linear Color Structure, ShadowOffset:double, Size:int=0, OutlineStrength:int=0)
   > n3 = Branch(Condition=@n2)
   n4 = Get DialogBody()
   > n5 = Set TextStyle(TextStyle:Text Block Style Structure=Get DefaultTextStyleOverride!(self:Rich Text Block Object Reference=@n4) -> (DefaultTextStyleOverride:Text Block Style Structure)) -> (Output_Get:Text Block Style Structure)
   n6 = Break Text Block Style!(TextBlockStyle:"Text Block Style Structure (by ref)"=@n5) -> (Font:Slate Font Info Structure) {is=title K2Node_BreakStruct}
   > n7 = Set members in Font Outline Settings!(
    execute:Exec
    StructRef:"Font Outline Settings Structure (by ref)"=Break Slate Font Info!(SlateFontInfo:"Slate Font Info Structure (by ref)"=@n6) -> (OutlineSettings:Font Outline Settings Structure, bForceMonospaced:bool=~, bMaterialIsStencil:bool=~, MonospacedWidth:float=~) {is=title K2Node_BreakStruct}
    OutlineSize:int=@n2.OutlineStrength
   ) -> (then:Exec, StructOut:"Font Outline Settings Structure (by ref)") {is=title K2Node_SetFieldsInStruct}
   > n8 = Set members in Slate Font Info!(
    execute:Exec
    StructRef:"Slate Font Info Structure (by ref)"=@n6
    OutlineSettings:Font Outline Settings Structure=@n7
    Size:float="To Float (Integer)"(InInt=@n2.Size)
   ) -> (then:Exec, StructOut:"Slate Font Info Structure (by ref)") {is=title K2Node_SetFieldsInStruct}
   > n9 = Set members in Text Block Style!(
    execute:Exec
    StructRef:"Text Block Style Structure (by ref)"=Get TextStyle() -> (TextStyle:Text Block Style Structure)
    Font:Slate Font Info Structure=@n8
    ColorAndOpacity:Slate Color Structure=Make Slate Color(SpecifiedColor=@n2.TextColor)
    ShadowOffset:Vector 2D Structure=MakeVector2D(X=@n2.ShadowOffset, Y=@n2.ShadowOffset)
   ) -> (then:Exec, StructOut:"Text Block Style Structure (by ref)") {is=title K2Node_SetFieldsInStruct}
   > SetDefaultTextStyle!(execute, self:Rich Text Block Object Reference=Get DialogBody(), InDefaultTextStyle:"Text Block Style Structure (by ref)"=@n9) -> (then)
   ClearAllDefaultStyleOverrides(execute=@n3.else, self:Rich Text Block Object Reference=@n4)
  ]}
 UpdateColor {nodes [
   n1 = UpdateColor!() -> (then, InUseLightColor:bool=~) {is=title K2Node_FunctionEntry}
   > n2 = Set bUseLightColorBackground(bUseLightColorBackground:bool=@n1) -> (Output_Get:bool)
   > SetBrushColor(
    self=Get background_Border() -> (background_Border:Border Object Reference)
    InBrushColor=Select(
     Option 0:Linear Color Structure=Get DarkColor() -> (DarkColor:Linear Color Structure)
     Option 1:Linear Color Structure=Get LightColor() -> (LightColor:Linear Color Structure)
     Index=@n2
    ) -> (ReturnValue:Linear Color Structure)
   )
  ]}
 UpdateEmptyPrompt {nodes [
   UpdateEmptyPrompt() {is=title K2Node_FunctionEntry}
   n1 = GetText!(self:Rich Text Block Object Reference=Get DialogBody()) -> (ReturnValue:Text)
   > SetVisibility(
    self=Get Prompt_Text() -> (Prompt_Text:Common Text Object Reference)
    InVisibility=Select(Option 0:ESlateVisibility Enum=Hidden, Option 1:ESlateVisibility Enum=SelfHitTestInvisible, Index=TextIsEmpty(InText=@n1)) -> (ReturnValue:ESlateVisibility Enum)
   )
  ]}
]
variables [
 bUseLightColorBackground {type=boolean}
 DarkColor {type=LinearColor, default=<rgba 0.0027 0.006 0.0452 1>}
 EditColorInputAction {type=DataTableRowHandle, default {dataTable=null, rowName=None}}
 EditTextInputAction {type=DataTableRowHandle, default {dataTable=&InputActionDataTable, rowName=Confirm}}
 LightColor {type=LinearColor, default=<rgba 0.9271 0.9271 0.9271 1>}
]
widgets {
 info {
  inheritedWidgetCount=0
  parentClass=/Script/FortniteUI.FortCreativePropertyEditorDialogBase
  rootWidgetClass=&Overlay
  widgetCount=22
 }
 widgets [
  Overlay_0 {widgetClassPath=&Overlay, widget=Overlay_0}
  SizeBox_WidgetSize {
   widgetClassPath=&SizeBox
   bIsVariable=true
   parent=Overlay_0
   widget=SizeBox_WidgetSize
   properties {
    bOverride_MaxDesiredHeight=true
    bOverride_MinDesiredHeight=true
    bOverride_WidthOverride=true
    heightOverride=715.217
    maxDesiredHeight=712
    minDesiredHeight=256
    widthOverride=640
   }
   slot {class=OverlaySlot}
  }
  background_Border {
   widgetClassPath=&Border
   bIsVariable=true
   parent=SizeBox_WidgetSize
   widget=background_Border
   properties {padding=<ltrb 18 18 18 18>}
   slot {class=SizeBoxSlot}
  }
  VerticalBox_0 {
   widgetClassPath=&VerticalBox
   parent=background_Border
   widget=VerticalBox_0
   slot {class=BorderSlot, properties {padding=<ltrb 18 18 18 18>}}
  }
  Overlay_2 {
   widgetClassPath=&Overlay
   parent=VerticalBox_0
   widget=Overlay_2
   slot {class=VerticalBoxSlot, properties {padding=<ltrb 0 0 0 12>, size {sizeRule=Fill, value=1}}}
  }
  ScrollBox_DialogBody {
   widgetClassPath=&ScrollBox
   bIsVariable=true
   parent=Overlay_2
   widget=ScrollBox_DialogBody
   properties {
    bAnimateWheelScrolling=true
    scrollWhenFocusChanges=AnimatedScroll
    scrollbarPadding=<ltrb 0 0 0 0>
    scrollbarThickness=<6 8>
    widgetBarStyle {
     thickness=6
     draggedThumbImage {
      bIsDynamicallyLoaded=false
      drawAs=Box
      imageType=FullColor
      mirroring=NoMirror
      resourceName=../../../Engine/Content/Slate/Common/Scrollbar_Thumb.png
      resourceObject=/Game/Athena/UI/ReplayBrowser/M_Ui_Skewed_Vfor9slices_BG_Separator_Inst
      tiling=NoTile
      imageSize=<0 0>
      margin=<ltrb 0.25 0.25 0.25 0.25>
      outlineSettings {
       bUseBrushTransparency=false
       roundingType=HalfHeightRadius
       width=0
       color {colorUseRule=UseColor_Specified, specifiedColor=<rgba 0 0 0 0>}
       cornerRadii=<0 0 0 0>
      }
      tintColor {colorUseRule=UseColor_Specified, specifiedColor=<rgba 0.6823 0.9101 1 1>}
      uVRegion {bIsValid=false, max=<0 0>, min=<0 0>}
     }
     horizontalBackgroundImage {
      bIsDynamicallyLoaded=false
      drawAs=NoDrawType
      imageType=NoImage
      mirroring=NoMirror
      resourceName=None
      resourceObject=null
      tiling=NoTile
      imageSize=<0 0>
      margin=<ltrb 0 0 0 0>
      outlineSettings {
       bUseBrushTransparency=false
       roundingType=HalfHeightRadius
       width=0
       color {colorUseRule=UseColor_Specified, specifiedColor=<rgba 0 0 0 0>}
       cornerRadii=<0 0 0 0>
      }
      tintColor {colorUseRule=UseColor_Specified, specifiedColor=<rgba 1 1 1 1>}
      uVRegion {bIsValid=false, max=<0 0>, min=<0 0>}
     }
     horizontalBottomSlotImage {
      bIsDynamicallyLoaded=false
      drawAs=NoDrawType
      imageType=FullColor
      mirroring=NoMirror
      resourceName=../../../Engine/Content/Slate/Common/Scrollbar_Background_Horizontal.png
      resourceObject=null
      tiling=NoTile
      imageSize=<8 8>
      margin=<ltrb 0 0 0 0>
      outlineSettings {
       bUseBrushTransparency=false
       roundingType=HalfHeightRadius
       width=0
       color {colorUseRule=UseColor_Specified, specifiedColor=<rgba 0 0 0 0>}
       cornerRadii=<0 0 0 0>
      }
      tintColor {colorUseRule=UseColor_Specified, specifiedColor=<rgba 1 1 1 1>}
      uVRegion {bIsValid=false, max=<0 0>, min=<0 0>}
     }
     horizontalTopSlotImage {
      bIsDynamicallyLoaded=false
      drawAs=NoDrawType
      imageType=FullColor
      mirroring=NoMirror
      resourceName=../../../Engine/Content/Slate/Common/Scrollbar_Background_Horizontal.png
      resourceObject=null
      tiling=NoTile
      imageSize=<8 8>
      margin=<ltrb 0 0 0 0>
      outlineSettings {
       bUseBrushTransparency=false
       roundingType=HalfHeightRadius
       width=0
       color {colorUseRule=UseColor_Specified, specifiedColor=<rgba 0 0 0 0>}
       cornerRadii=<0 0 0 0>
      }
      tintColor {colorUseRule=UseColor_Specified, specifiedColor=<rgba 1 1 1 1>}
      uVRegion {bIsValid=false, max=<0 0>, min=<0 0>}
     }
     hoveredThumbImage {
      bIsDynamicallyLoaded=false
      drawAs=Box
      imageType=FullColor
      mirroring=NoMirror
      resourceName=../../../Engine/Content/Slate/Common/Scrollbar_Thumb.png
      resourceObject=&M_Ui_Skewed_Vfor9slices_BG_Separator_Inst
      tiling=NoTile
      imageSize=<0 0>
      margin=<ltrb 0.25 0.25 0.25 0.25>
      outlineSettings {
       bUseBrushTransparency=false
       roundingType=HalfHeightRadius
       width=0
       color {colorUseRule=UseColor_Specified, specifiedColor=<rgba 0 0 0 0>}
       cornerRadii=<0 0 0 0>
      }
      tintColor {colorUseRule=UseColor_Specified, specifiedColor=<rgba 0.3941 0.736 1 1>}
      uVRegion {bIsValid=false, max=<0 0>, min=<0 0>}
     }
     normalThumbImage {
      bIsDynamicallyLoaded=false
      drawAs=Box
      imageType=FullColor
      mirroring=NoMirror
      resourceName=../../../Engine/Content/Slate/Common/Scrollbar_Thumb.png
      resourceObject=&M_Ui_Skewed_Vfor9slices_BG_Separator_Inst
      tiling=NoTile
      imageSize=<0 0>
      margin=<ltrb 0.25 0.25 0.25 0.25>
      outlineSettings {
       bUseBrushTransparency=false
       roundingType=HalfHeightRadius
       width=0
       color {colorUseRule=UseColor_Specified, specifiedColor=<rgba 0 0 0 0>}
       cornerRadii=<0 0 0 0>
      }
      tintColor {colorUseRule=UseColor_Specified, specifiedColor=<rgba 1 1 1 1>}
      uVRegion {bIsValid=false, max=<0 0>, min=<0 0>}
     }
     verticalBackgroundImage {
      bIsDynamicallyLoaded=false
      drawAs=Image
      imageType=NoImage
      mirroring=NoMirror
      resourceName=None
      resourceObject=&M_Ui_Skewed_Vfor9slices_BG_Separator_Inst
      tiling=NoTile
      imageSize=<0 0>
      margin=<ltrb 0 0 0 0>
      outlineSettings {
       bUseBrushTransparency=false
       roundingType=HalfHeightRadius
       width=0
       color {colorUseRule=UseColor_Specified, specifiedColor=<rgba 0 0 0 0>}
       cornerRadii=<0 0 0 0>
      }
      tintColor {colorUseRule=UseColor_Specified, specifiedColor=<rgba 1 1 1 0.25>}
      uVRegion {bIsValid=false, max=<0 0>, min=<0 0>}
     }
     verticalBottomSlotImage {
      bIsDynamicallyLoaded=false
      drawAs=NoDrawType
      imageType=FullColor
      mirroring=NoMirror
      resourceName=../../../Engine/Content/Slate/Common/Scrollbar_Background_Vertical.png
      resourceObject=null
      tiling=NoTile
      imageSize=<8 8>
      margin=<ltrb 0 0 0 0>
      outlineSettings {
       bUseBrushTransparency=false
       roundingType=HalfHeightRadius
       width=0
       color {colorUseRule=UseColor_Specified, specifiedColor=<rgba 0 0 0 0>}
       cornerRadii=<0 0 0 0>
      }
      tintColor {colorUseRule=UseColor_Specified, specifiedColor=<rgba 1 1 1 1>}
      uVRegion {bIsValid=false, max=<0 0>, min=<0 0>}
     }
     verticalTopSlotImage {
      bIsDynamicallyLoaded=false
      drawAs=NoDrawType
      imageType=FullColor
      mirroring=NoMirror
      resourceName=../../../Engine/Content/Slate/Common/Scrollbar_Background_Vertical.png
      resourceObject=null
      tiling=NoTile
      imageSize=<8 8>
      margin=<ltrb 0 0 0 0>
      outlineSettings {
       bUseBrushTransparency=false
       roundingType=HalfHeightRadius
       width=0
       color {colorUseRule=UseColor_Specified, specifiedColor=<rgba 0 0 0 0>}
       cornerRadii=<0 0 0 0>
      }
      tintColor {colorUseRule=UseColor_Specified, specifiedColor=<rgba 1 1 1 1>}
      uVRegion {bIsValid=false, max=<0 0>, min=<0 0>}
     }
    }
    widgetStyle {
     barThickness=9
     bottomShadowBrush {
      bIsDynamicallyLoaded=false
      drawAs=NoDrawType
      imageType=FullColor
      mirroring=NoMirror
      resourceName=../../../Engine/Content/Slate/Common/ScrollBoxShadowBottom.png
      resourceObject=null
      tiling=NoTile
      imageSize=<16 8>
      margin=<ltrb 0.5 0 0.5 1>
      outlineSettings {
       bUseBrushTransparency=false
       roundingType=HalfHeightRadius
       width=0
       color {colorUseRule=UseColor_Specified, specifiedColor=<rgba 0 0 0 0>}
       cornerRadii=<0 0 0 0>
      }
      tintColor {colorUseRule=UseColor_Specified, specifiedColor=<rgba 1 1 1 1>}
      uVRegion {bIsValid=false, max=<0 0>, min=<0 0>}
     }
     horizontalScrolledContentPadding=<ltrb 0 0 1 0>
     leftShadowBrush {
      bIsDynamicallyLoaded=false
      drawAs=Box
      imageType=FullColor
      mirroring=NoMirror
      resourceName=../../../Engine/Content/Slate/Common/ScrollBoxShadowLeft.png
      resourceObject=null
      tiling=NoTile
      imageSize=<8 16>
      margin=<ltrb 1 0.5 0 0.5>
      outlineSettings {
       bUseBrushTransparency=false
       roundingType=HalfHeightRadius
       width=0
       color {colorUseRule=UseColor_Specified, specifiedColor=<rgba 0 0 0 0>}
       cornerRadii=<0 0 0 0>
      }
      tintColor {colorUseRule=UseColor_Specified, specifiedColor=<rgba 1 1 1 1>}
      uVRegion {bIsValid=false, max=<0 0>, min=<0 0>}
     }
     rightShadowBrush {
      bIsDynamicallyLoaded=false
      drawAs=Box
      imageType=FullColor
      mirroring=NoMirror
      resourceName=../../../Engine/Content/Slate/Common/ScrollBoxShadowRight.png
      resourceObject=null
      tiling=NoTile
      imageSize=<8 16>
      margin=<ltrb 0 0.5 1 0.5>
      outlineSettings {
       bUseBrushTransparency=false
       roundingType=HalfHeightRadius
       width=0
       color {colorUseRule=UseColor_Specified, specifiedColor=<rgba 0 0 0 0>}
       cornerRadii=<0 0 0 0>
      }
      tintColor {colorUseRule=UseColor_Specified, specifiedColor=<rgba 1 1 1 1>}
      uVRegion {bIsValid=false, max=<0 0>, min=<0 0>}
     }
     topShadowBrush {
      bIsDynamicallyLoaded=false
      drawAs=NoDrawType
      imageType=FullColor
      mirroring=NoMirror
      resourceName=../../../Engine/Content/Slate/Common/ScrollBoxShadowTop.png
      resourceObject=null
      tiling=NoTile
      imageSize=<16 8>
      margin=<ltrb 0.5 1 0.5 0>
      outlineSettings {
       bUseBrushTransparency=false
       roundingType=HalfHeightRadius
       width=0
       color {colorUseRule=UseColor_Specified, specifiedColor=<rgba 0 0 0 0>}
       cornerRadii=<0 0 0 0>
      }
      tintColor {colorUseRule=UseColor_Specified, specifiedColor=<rgba 1 1 1 1>}
      uVRegion {bIsValid=false, max=<0 0>, min=<0 0>}
     }
     verticalScrolledContentPadding=<ltrb 0 0 0 1>
    }
   }
   slot {class=OverlaySlot, properties {horizontalAlignment=HAlign_Fill, verticalAlignment=VAlign_Center}}
  }
  Overlay_4 {widgetClassPath=&Overlay, parent=ScrollBox_DialogBody, widget=Overlay_4, slot {class=ScrollBoxSlot}}
  DialogBody {
   widgetClassPath=&CommonRichTextBlock
   bIsVariable=true
   parent=Overlay_4
   widget=DialogBody
   properties {
    autoWrapText=true
    justification=Center
    wrappingPolicy=AllowPerCharacterWrapping
    defaultTextStyleOverride {
     overflowPolicy=Clip
     transformPolicy=None
     colorAndOpacity {colorUseRule=UseColor_Specified, specifiedColor=<rgba 1 1 1 1>}
     font {
      bForceMonospaced=false
      bMaterialIsStencil=false
      fontMaterial=null
      fontObject=&BurbankSmall
      letterSpacing=0
      monospacedWidth=1
      size=22
      skewAmount=0
      typefaceFontName=Bold
      outlineSettings {
       bApplyOutlineToDropShadows=false
       bMiteredCorners=false
       bSeparateFillAlpha=false
       outlineMaterial=null
       outlineSize=0
       outlineColor=<rgba 0 0 0 1>
      }
     }
     highlightColor {colorUseRule=UseColor_Specified, specifiedColor=<rgba 0 0 0 1>}
     highlightShape {
      bIsDynamicallyLoaded=false
      drawAs=Image
      imageType=NoImage
      mirroring=NoMirror
      resourceName=None
      resourceObject=null
      tiling=NoTile
      imageSize=<32 32>
      margin=<ltrb 0 0 0 0>
      outlineSettings {
       bUseBrushTransparency=false
       roundingType=HalfHeightRadius
       width=0
       color {colorUseRule=UseColor_Specified, specifiedColor=<rgba 0 0 0 0>}
       cornerRadii=<0 0 0 0>
      }
      tintColor {colorUseRule=UseColor_Specified, specifiedColor=<rgba 1 1 1 1>}
      uVRegion {bIsValid=false, max=<0 0>, min=<0 0>}
     }
     selectedBackgroundColor {colorUseRule=UseColor_Foreground, specifiedColor=<rgba 1 0 1 1>}
     shadowColorAndOpacity=<rgba 0 0 0 1>
     shadowOffset=<0 0>
     strikeBrush {
      bIsDynamicallyLoaded=false
      drawAs=Image
      imageType=NoImage
      mirroring=NoMirror
      resourceName=None
      resourceObject=null
      tiling=NoTile
      imageSize=<32 32>
      margin=<ltrb 0 0 0 0>
      outlineSettings {
       bUseBrushTransparency=false
       roundingType=HalfHeightRadius
       width=0
       color {colorUseRule=UseColor_Specified, specifiedColor=<rgba 0 0 0 0>}
       cornerRadii=<0 0 0 0>
      }
      tintColor {colorUseRule=UseColor_Specified, specifiedColor=<rgba 1 1 1 1>}
      uVRegion {bIsValid=false, max=<0 0>, min=<0 0>}
     }
     underlineBrush {
      bIsDynamicallyLoaded=false
      drawAs=Image
      imageType=NoImage
      mirroring=NoMirror
      resourceName=None
      resourceObject=null
      tiling=NoTile
      imageSize=<32 32>
      margin=<ltrb 0 0 0 0>
      outlineSettings {
       bUseBrushTransparency=false
       roundingType=HalfHeightRadius
       width=0
       color {colorUseRule=UseColor_Specified, specifiedColor=<rgba 0 0 0 0>}
       cornerRadii=<0 0 0 0>
      }
      tintColor {colorUseRule=UseColor_Specified, specifiedColor=<rgba 1 1 1 1>}
      uVRegion {bIsValid=false, max=<0 0>, min=<0 0>}
     }
    }
   }
   slot {class=OverlaySlot, properties {horizontalAlignment=HAlign_Fill, verticalAlignment=VAlign_Fill}}
  }
  Prompt_Text {
   widgetClassPath=&CommonTextBlock
   bIsVariable=true
   parent=Overlay_2
   widget=Prompt_Text
   properties {
    autoWrapText=true
    justification=Center
    text=Select to add or modify text.
    visibility=Hidden
    colorAndOpacity {colorUseRule=UseColor_Specified, specifiedColor=<rgba 1 1 1 0.5>}
    font {
     bForceMonospaced=false
     bMaterialIsStencil=false
     fontMaterial=null
     fontObject=&BurbankSmall
     letterSpacing=0
     monospacedWidth=1
     size=18
     skewAmount=0
     typefaceFontName=Bold
     outlineSettings {
      bApplyOutlineToDropShadows=false
      bMiteredCorners=false
      bSeparateFillAlpha=false
      outlineMaterial=null
      outlineSize=0
      outlineColor=<rgba 0 0 0 1>
     }
    }
    shadowOffset=<0 0>
   }
   slot {class=OverlaySlot, properties {horizontalAlignment=HAlign_Fill, verticalAlignment=VAlign_Center}}
  }
  Overlay_3 {
   widgetClassPath=&Overlay
   parent=VerticalBox_0
   widget=Overlay_3
   slot {class=VerticalBoxSlot, properties {verticalAlignment=VAlign_Bottom}}
  }
  HorizontalBox_1 {
   widgetClassPath=&HorizontalBox
   parent=Overlay_3
   widget=HorizontalBox_1
   slot {class=OverlaySlot, properties {horizontalAlignment=HAlign_Center}}
  }
  CommonHardwareVisibilityBorder_0 {
   widgetClassPath=&CommonHardwareVisibilityBorder
   parent=HorizontalBox_1
   widget=CommonHardwareVisibilityBorder_0
   properties {
    horizontalAlignment=HAlign_Center
    visibilityQuery {
     autoDescription=" ANY( Input.Gamepad )"
     tokenStreamVersion=0
     userDescription=Gamepad
     queryTokenStream [0,1,1,1,0]
     tagDictionary [{tagName=Input.Gamepad}]
    }
   }
   slot {class=HorizontalBoxSlot}
  }
  HorizontalBox_75 {
   widgetClassPath=&HorizontalBox
   parent=CommonHardwareVisibilityBorder_0
   widget=HorizontalBox_75
   slot {class=BorderSlot, properties {horizontalAlignment=HAlign_Center, padding=<ltrb 0 0 0 0>}}
  }
  EditTextActionWidget {
   widgetClassPath=&CommonActionWidget
   bIsVariable=true
   parent=HorizontalBox_75
   widget=EditTextActionWidget
   properties {visibility=SelfHitTestInvisible}
   slot {class=HorizontalBoxSlot, properties {verticalAlignment=VAlign_Center, padding=<ltrb 0 0 8 0>}}
  }
  CommonTextBlock_60 {
   widgetClassPath=&CommonTextBlock
   parent=HorizontalBox_75
   widget=CommonTextBlock_60
   properties {
    scrollStyle=&TextScrollStyle-Base_C
    text=Edit
    font {
     bForceMonospaced=false
     bMaterialIsStencil=false
     fontMaterial=null
     fontObject=&BurbankSmall
     letterSpacing=0
     monospacedWidth=1
     size=14
     skewAmount=0
     typefaceFontName=Bold
     outlineSettings {
      bApplyOutlineToDropShadows=false
      bMiteredCorners=false
      bSeparateFillAlpha=false
      outlineMaterial=null
      outlineSize=1
      outlineColor=<rgba 0 0 0 0.3>
     }
    }
    shadowOffset=<0 0>
   }
   slot {class=HorizontalBoxSlot, properties {verticalAlignment=VAlign_Center}}
  }
  Spacer_297 {widgetClassPath=&Spacer, parent=HorizontalBox_1, widget=Spacer_297, properties {size=<32 1>}, slot {class=HorizontalBoxSlot}}
  CommonHardwareVisibilityBorder_1 {
   widgetClassPath=&CommonHardwareVisibilityBorder
   parent=HorizontalBox_1
   widget=CommonHardwareVisibilityBorder_1
   properties {bIsEnabled=false, visibility=Collapsed}
   slot {class=HorizontalBoxSlot}
  }
  HorizontalBox_0 {
   widgetClassPath=&HorizontalBox
   parent=CommonHardwareVisibilityBorder_1
   widget=HorizontalBox_0
   properties {bIsEnabled=false, visibility=Collapsed}
   slot {class=BorderSlot, properties {padding=<ltrb 0 0 0 0>}}
  }
  EditColorActionWidget {
   widgetClassPath=&CommonActionWidget
   bIsVariable=true
   parent=HorizontalBox_0
   widget=EditColorActionWidget
   properties {bIsEnabled=false, visibility=Collapsed}
   slot {class=HorizontalBoxSlot, properties {padding=<ltrb 0 0 8 0>}}
  }
  Overlay_1 {widgetClassPath=&Overlay, parent=HorizontalBox_0, widget=Overlay_1, slot {class=HorizontalBoxSlot}}
  CommonBorder_47 {
   widgetClassPath=&CommonBorder
   parent=Overlay_1
   widget=CommonBorder_47
   properties {
    background {
     bIsDynamicallyLoaded=false
     drawAs=RoundedBox
     imageType=NoImage
     mirroring=NoMirror
     resourceName=None
     resourceObject=null
     tiling=NoTile
     imageSize=<32 32>
     margin=<ltrb 0 0 0 0>
     outlineSettings {
      bUseBrushTransparency=false
      roundingType=FixedRadius
      width=1
      color {colorUseRule=UseColor_Specified, specifiedColor=<rgba 1 1 1 1>}
      cornerRadii=<2 2 2 2>
     }
     tintColor {colorUseRule=UseColor_Specified, specifiedColor=<rgba 0.0833 0.0833 0.0833 1>}
     uVRegion {bIsValid=false, max=<0 0>, min=<0 0>}
    }
    padding=<ltrb 12 12 12 12>
   }
   slot {class=OverlaySlot}
  }
  CommonBorder_108 {
   widgetClassPath=&CommonBorder
   parent=Overlay_1
   widget=CommonBorder_108
   properties {
    background {
     bIsDynamicallyLoaded=false
     drawAs=RoundedBox
     imageType=NoImage
     mirroring=NoMirror
     resourceName=None
     resourceObject=null
     tiling=NoTile
     imageSize=<32 32>
     margin=<ltrb 0 0 0 0>
     outlineSettings {
      bUseBrushTransparency=false
      roundingType=FixedRadius
      width=1
      color {colorUseRule=UseColor_Specified, specifiedColor=<rgba 0 0 0 1>}
      cornerRadii=<2 2 2 2>
     }
     tintColor {colorUseRule=UseColor_Specified, specifiedColor=<rgba 1 1 1 1>}
     uVRegion {bIsValid=false, max=<0 0>, min=<0 0>}
    }
    padding=<ltrb 12 12 12 12>
   }
   slot {class=OverlaySlot, properties {padding=<ltrb 8 8 0 0>}}
  }
 ]
}
