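# Selectable rich-text word button for the Creative property editor's Style Editor, derived from FortCreativeRichTextWordButton.
# Select, deselect and hover events play forward or reverse animations on three stacked Borders for hover, selected and selected-hover states, swapping brushes from VisualState variables.
# UpdateDefaultStyle queries a text style override interface, then rebuilds the text's default color, size, outline and shadow offset, or clears overrides when none exists.
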WidgetBlueprint /FortniteCoreUI/InGame/Creative/PropertyEditor/StyleEditor/StyleEditorWordItem
animations [
 OnHover {
  binding_count=1
  length_seconds=0.25
  object=&self:OnHover
  sequence {bindings [Border_hover {
     id=#1
     possessed_class=Border
     tracks [
      {
       type=MovieScene2DTransformTrack
       sections [{
         type=MovieScene2DTransformSection
         end_frame=5
         start_frame=0
         channels [
          Translation.X {type=MovieSceneScriptingFloatChannel, const [0,2]}
          Translation.Y {type=MovieSceneScriptingFloatChannel, const [0,2]}
          Scale.X {type=MovieSceneScriptingFloatChannel, const [1,2]}
          Scale.Y {type=MovieSceneScriptingFloatChannel, keys [[0,0.8],[3,2],[5,1]]}
         ]
        }]
      }
      {
       type=MovieSceneFloatTrack
       display_name=Render Opacity
       sections [{type=MovieSceneFloatSection, end_frame=5, start_frame=0, channels [None {type=MovieSceneScriptingFloatChannel, keys [[0,0],[3,0.8],[5,1]]}]}]
      }
     ]
    }]}
 }
 OnSelected {
  binding_count=1
  length_seconds=0.25
  object=&self:OnSelected
  sequence {bindings [Border_selected {
     id=#2
     possessed_class=Border
     tracks [
      {
       type=MovieScene2DTransformTrack
       sections [{
         type=MovieScene2DTransformSection
         end_frame=5
         start_frame=0
         channels [
          Translation.X {type=MovieSceneScriptingFloatChannel, const [0,1]}
          Translation.Y {type=MovieSceneScriptingFloatChannel, const [0,1]}
          Scale.X {type=MovieSceneScriptingFloatChannel, keys [[0,1],[3,1.1],[5,1]]}
          Scale.Y {type=MovieSceneScriptingFloatChannel, keys [[0,1],[3,1.1],[5,1]]}
         ]
        }]
      }
      {
       type=MovieSceneFloatTrack
       display_name=Render Opacity
       sections [{type=MovieSceneFloatSection, end_frame=5, start_frame=0, channels [None {type=MovieSceneScriptingFloatChannel, keys [[0,0],[5,0.8]]}]}]
      }
     ]
    }]}
 }
 OnHoverwhenSelected {
  binding_count=1
  length_seconds=0.25
  object=&self:OnHoverwhenSelected
  sequence {bindings [Border_selectedandhovering {
     id=#3
     possessed_class=Border
     tracks [
      {
       type=MovieSceneFloatTrack
       display_name=Render Opacity
       sections [{type=MovieSceneFloatSection, end_frame=5, start_frame=0, channels [None {type=MovieSceneScriptingFloatChannel, keys [[0,0],[5,1]]}]}]
      }
      {
       type=MovieScene2DTransformTrack
       sections [{
         type=MovieScene2DTransformSection
         end_frame=5
         start_frame=0
         channels [
          Translation.X {type=MovieSceneScriptingFloatChannel, const [0,2]}
          Translation.Y {type=MovieSceneScriptingFloatChannel, const [0,2]}
          Angle {type=MovieSceneScriptingFloatChannel, const [0,1]}
          Scale.X {type=MovieSceneScriptingFloatChannel, const [1,2]}
          Scale.Y {type=MovieSceneScriptingFloatChannel, const [1,2]}
          Shear.X {type=MovieSceneScriptingFloatChannel, const [0,1]}
          Shear.Y {type=MovieSceneScriptingFloatChannel, const [0,1]}
         ]
        }]
      }
     ]
    }]}
 }
]
graphs [
 EventGraph {nodes [
   Event On Selected()
   > PlayAnimationForward(InAnimation=Get OnSelected() -> (OnSelected:Widget Animation Object Reference))
   > SetBrush(
    self:Border Object Reference=Get Border_selected() -> (Border_selected:Border Object Reference)
    InBrush=Get VisualState_Selected_NotHovered() -> (VisualState_Selected_NotHovered:Slate Brush Structure)
   )
   > PlayAnimationForward(InAnimation=Get OnHoverwhenSelected() -> (OnHoverwhenSelected:Widget Animation Object Reference))
   > SetBrush(
    self=Get Border_selectedandhovering() -> (Border_selectedandhovering:Border Object Reference)
    InBrush=Get VisualState_Selected_Hovered() -> (VisualState_Selected_Hovered:Slate Brush Structure)
   )
   Event OnButtonInitialized() {is=title K2Node_Event}
   n1 = Get VisualState_NotSelected_NotHovered() -> (VisualState_NotSelected_NotHovered:Slate Brush Structure)
   > SetBrush(self=Get Border_selected(), InBrush=@n1)
   > SetBrush(self=Get Border_selectedandhovering(), InBrush=@n1)
   Event On Deselected()
   > PlayAnimationReverse(InAnimation=Get OnSelected())
   > PlayAnimationReverse(InAnimation=Get OnHoverwhenSelected())
   n2 = Get VisualState_NotSelected_NotHovered()
   > SetBrush(self=Get Border_selected(), InBrush=@n2)
   > SetBrush(self=Get Border_selectedandhovering(), InBrush=@n2)
   > n3 = Branch(Condition=IsHovered())
   SetRenderOpacity(execute=@n3.else, self=Get Border_hover() -> (Border_hover:Border Object Reference), InOpacity=0)
   Event On Hovered()
   > n4 = Branch(Condition=GetSelected!(self:Common Button Base Object Reference) -> (ReturnValue:bool))
   > PlayAnimationForward(InAnimation=Get OnHoverwhenSelected())
   > SetBrush(self=Get Border_selectedandhovering(), InBrush=Get VisualState_Selected_Hovered())
   PlayAnimationForward(execute=@n4.else, InAnimation=Get OnHover() -> (OnHover:Widget Animation Object Reference))
   Event On Unhovered()
   > n5 = Branch(Condition=GetSelected())
   > PlayAnimationReverse(InAnimation=Get OnHoverwhenSelected())
   > SetBrush(self=Get Border_selectedandhovering(), InBrush=Get VisualState_NotSelected_NotHovered())
   PlayAnimationReverse(execute=@n5.else, InAnimation=Get OnHover())
   Event Construct()
  ]}
 Focus {nodes [
   Focus() {is=title K2Node_FunctionEntry}
   > SetFocus()
  ]}
 UpdateDefaultStyle {nodes [
   n1 = UpdateDefaultStyle!() -> (then, Target:BPi Text Style Overrides Interface) {is=title K2Node_FunctionEntry}
   > n2 = GetDefaultStyleOverride!(execute, self:BPi Text Style Overrides Interface=@n1) -> (then, HasOverride:bool, TextColor:Linear Color Structure, ShadowOffset:double, Size:int=0, OutlineStrength:int=0)
   > n3 = Branch(Condition=@n2)
   n4 = Get DefaultTextStyleOverride!(self:Rich Text Block Object Reference=Get Rich_Text() -> (Rich_Text:Rich Text Block Object Reference)) -> (DefaultTextStyleOverride:Text Block Style Structure)
   n5 = Break Text Block Style!(TextBlockStyle:"Text Block Style Structure (by ref)"=@n4) -> (Font:Slate Font Info Structure) {is=title K2Node_BreakStruct}
   > n6 = Set members in Font Outline Settings!(
    execute:Exec
    StructRef:"Font Outline Settings Structure (by ref)"=Break Slate Font Info!(SlateFontInfo:"Slate Font Info Structure (by ref)"=@n5) -> (OutlineSettings:Font Outline Settings Structure, bForceMonospaced:bool=~, bMaterialIsStencil:bool=~, MonospacedWidth:float=~) {is=title K2Node_BreakStruct}
    OutlineSize:int=@n2.OutlineStrength
   ) -> (then:Exec, StructOut:"Font Outline Settings Structure (by ref)") {is=title K2Node_SetFieldsInStruct}
   > n7 = Set members in Slate Font Info!(
    execute:Exec
    StructRef:"Slate Font Info Structure (by ref)"=@n5
    OutlineSettings:Font Outline Settings Structure=@n6
    Size:float="To Float (Integer)"(InInt=@n2.Size)
   ) -> (then:Exec, StructOut:"Slate Font Info Structure (by ref)") {is=title K2Node_SetFieldsInStruct}
   > n8 = Set members in Text Block Style!(
    execute:Exec
    StructRef:"Text Block Style Structure (by ref)"=@n4
    Font:Slate Font Info Structure=@n7
    ColorAndOpacity:Slate Color Structure=Make Slate Color(SpecifiedColor=@n2.TextColor)
    ShadowOffset:Vector 2D Structure=MakeVector2D(X=@n2.ShadowOffset, Y=@n2.ShadowOffset)
   ) -> (then:Exec, StructOut:"Text Block Style Structure (by ref)") {is=title K2Node_SetFieldsInStruct}
   > SetDefaultTextStyle!(execute, self:Rich Text Block Object Reference=Get Rich_Text(), InDefaultTextStyle:"Text Block Style Structure (by ref)"=@n8) -> (then)
   ClearAllDefaultStyleOverrides(execute=@n3.else, self:Rich Text Block Object Reference=Get Rich_Text())
  ]}
]
variables [
 VisualState_NotSelected_Hovered {
  type=SlateBrush
  default {
   bIsDynamicallyLoaded=false
   drawAs=RoundedBox
   imageType=NoImage
   mirroring=NoMirror
   resourceName=None
   resourceObject=null
   tiling=NoTile
   imageSize=<32 32>
   margin=<ltrb 0 0 0 0>
   outlineSettings {
    bUseBrushTransparency=false
    roundingType=FixedRadius
    width=2
    color {colorUseRule=UseColor_Specified, specifiedColor=<rgba 1 1 1 1>}
    cornerRadii=<6 6 6 6>
   }
   tintColor {colorUseRule=UseColor_Specified, specifiedColor=<rgba 1 1 1 0>}
   uVRegion {bIsValid=false, max=<0 0>, min=<0 0>}
  }
 }
 VisualState_NotSelected_NotHovered {
  type=SlateBrush
  default {
   bIsDynamicallyLoaded=false
   drawAs=RoundedBox
   imageType=NoImage
   mirroring=NoMirror
   resourceName=None
   resourceObject=null
   tiling=NoTile
   imageSize=<32 32>
   margin=<ltrb 0 0 0 0>
   outlineSettings {
    bUseBrushTransparency=false
    roundingType=FixedRadius
    width=0
    color {colorUseRule=UseColor_Specified, specifiedColor=<rgba 0 0 0 0>}
    cornerRadii=<6 6 6 6>
   }
   tintColor {colorUseRule=UseColor_Specified, specifiedColor=<rgba 1 0.99 0.9896 0>}
   uVRegion {bIsValid=false, max=<0 0>, min=<0 0>}
  }
 }
 VisualState_Selected_Hovered {
  type=SlateBrush
  default {
   bIsDynamicallyLoaded=false
   drawAs=RoundedBox
   imageType=NoImage
   mirroring=NoMirror
   resourceName=None
   resourceObject=null
   tiling=NoTile
   imageSize=<32 32>
   margin=<ltrb 0 0 0 0>
   outlineSettings {
    bUseBrushTransparency=false
    roundingType=FixedRadius
    width=2
    color {colorUseRule=UseColor_Specified, specifiedColor=<rgba 0.0024 0.0513 0.3864 1>}
    cornerRadii=<0 0 0 0>
   }
   tintColor {colorUseRule=UseColor_Specified, specifiedColor=<rgba 1 1 1 1>}
   uVRegion {bIsValid=false, max=<0 0>, min=<0 0>}
  }
 }
 VisualState_Selected_NotHovered {
  type=SlateBrush
  default {
   bIsDynamicallyLoaded=false
   drawAs=RoundedBox
   imageType=NoImage
   mirroring=NoMirror
   resourceName=None
   resourceObject=null
   tiling=NoTile
   imageSize=<32 32>
   margin=<ltrb 0 0 0 0>
   outlineSettings {
    bUseBrushTransparency=false
    roundingType=FixedRadius
    width=2
    color {colorUseRule=UseColor_Specified, specifiedColor=<rgba 0.0044 0.3712 0.8308 1>}
    cornerRadii=<0 0 0 0>
   }
   tintColor {colorUseRule=UseColor_Specified, specifiedColor=<rgba 1 1 1 1>}
   uVRegion {bIsValid=false, max=<0 0>, min=<0 0>}
  }
 }
]
widgets {
 info {inheritedWidgetCount=2, parentClass=/Script/FortniteUI.FortCreativeRichTextWordButton, rootWidgetClass=&SizeBox, widgetCount=7}
 widgets [
  SizeBox_1 {
   widgetClassPath=&SizeBox
   widget=SizeBox_1
   properties {bOverride_MinDesiredHeight=true, bOverride_MinDesiredWidth=true, minDesiredHeight=30, minDesiredWidth=50}
  }
  Overlay_0 {widgetClassPath=&Overlay, parent=SizeBox_1, widget=Overlay_0, slot {class=SizeBoxSlot, properties {padding=<ltrb 1 1 1 1>}}}
  Border_hover {
   widgetClassPath=&Border
   bIsVariable=true
   parent=Overlay_0
   widget=Border_hover
   properties {
    renderOpacity=0
    visibility=SelfHitTestInvisible
    background {
     bIsDynamicallyLoaded=false
     drawAs=RoundedBox
     imageType=NoImage
     mirroring=NoMirror
     resourceName=None
     resourceObject=null
     tiling=NoTile
     imageSize=<32 32>
     margin=<ltrb 0 0 0 0>
     outlineSettings {
      bUseBrushTransparency=true
      roundingType=FixedRadius
      width=0.4
      color {colorUseRule=UseColor_Specified, specifiedColor=<rgba 1 1 1 1>}
      cornerRadii=<0 0 0 0>
     }
     tintColor {colorUseRule=UseColor_Specified, specifiedColor=<rgba 1 1 1 1>}
     uVRegion {bIsValid=false, max=<0 0>, min=<0 0>}
    }
    brushColor=<rgba 0.0044 0.3712 0.8308 1>
   }
   slot {class=OverlaySlot, properties {horizontalAlignment=HAlign_Fill, verticalAlignment=VAlign_Bottom}}
  }
  Border_selected {
   widgetClassPath=&Border
   bIsVariable=true
   parent=Overlay_0
   widget=Border_selected
   properties {
    renderOpacity=0
    visibility=HitTestInvisible
    background {
     bIsDynamicallyLoaded=false
     drawAs=RoundedBox
     imageType=NoImage
     mirroring=NoMirror
     resourceName=None
     resourceObject=null
     tiling=NoTile
     imageSize=<32 32>
     margin=<ltrb 0 0 0 0>
     outlineSettings {
      bUseBrushTransparency=false
      roundingType=FixedRadius
      width=2
      color {colorUseRule=UseColor_Specified, specifiedColor=<rgba 0.0044 0.3712 0.8308 1>}
      cornerRadii=<0 0 0 0>
     }
     tintColor {colorUseRule=UseColor_Specified, specifiedColor=<rgba 1 1 1 1>}
     uVRegion {bIsValid=false, max=<0 0>, min=<0 0>}
    }
    brushColor=<rgba 0.0024 0.0513 0.3864 0>
   }
   slot {class=OverlaySlot, properties {horizontalAlignment=HAlign_Fill, verticalAlignment=VAlign_Fill}}
  }
  Border_selectedandhovering {
   widgetClassPath=&Border
   bIsVariable=true
   parent=Overlay_0
   widget=Border_selectedandhovering
   properties {
    renderOpacity=0
    visibility=HitTestInvisible
    background {
     bIsDynamicallyLoaded=false
     drawAs=RoundedBox
     imageType=NoImage
     mirroring=NoMirror
     resourceName=None
     resourceObject=null
     tiling=NoTile
     imageSize=<32 32>
     margin=<ltrb 0 0 0 0>
     outlineSettings {
      bUseBrushTransparency=false
      roundingType=FixedRadius
      width=2
      color {colorUseRule=UseColor_Specified, specifiedColor=<rgba 1 1 1 1>}
      cornerRadii=<0 0 0 0>
     }
     tintColor {colorUseRule=UseColor_Specified, specifiedColor=<rgba 1 1 1 1>}
     uVRegion {bIsValid=false, max=<0 0>, min=<0 0>}
    }
    brushColor=<rgba 0.0025 0.211 0.4722 0>
   }
   slot {class=OverlaySlot, properties {horizontalAlignment=HAlign_Fill, verticalAlignment=VAlign_Fill}}
  }
  Rich_Text {
   widgetClassPath=/Script/UMG.RichTextBlock
   bInherited=true
   bIsVariable=true
   parent=Overlay_0
   widget=Rich_Text
   properties {
    bOverrideDefaultStyle=true
    textStyleSet=/Game/Athena/UI/InGame/Creative/PropertyEditor/StyleEditor/RichTextStyleSets/CreativeDefaultRichTextStyles
    defaultTextStyleOverride {
     overflowPolicy=Clip
     transformPolicy=None
     colorAndOpacity {colorUseRule=UseColor_Specified, specifiedColor=<rgba 0.8746 1 0.913 1>}
     font {
      bForceMonospaced=false
      bMaterialIsStencil=false
      fontMaterial=null
      fontObject=&BurbankSmall
      letterSpacing=0
      monospacedWidth=1
      size=18
      skewAmount=0
      typefaceFontName=Bold
      outlineSettings {
       bApplyOutlineToDropShadows=false
       bMiteredCorners=false
       bSeparateFillAlpha=false
       outlineMaterial=null
       outlineSize=0
       outlineColor=<rgba 0 0 0 1>
      }
     }
     highlightColor {colorUseRule=UseColor_Specified, specifiedColor=<rgba 0 0 0 1>}
     highlightShape {
      bIsDynamicallyLoaded=false
      drawAs=Image
      imageType=NoImage
      mirroring=NoMirror
      resourceName=None
      resourceObject=null
      tiling=NoTile
      imageSize=<32 32>
      margin=<ltrb 0 0 0 0>
      outlineSettings {
       bUseBrushTransparency=false
       roundingType=HalfHeightRadius
       width=0
       color {colorUseRule=UseColor_Specified, specifiedColor=<rgba 0 0 0 0>}
       cornerRadii=<0 0 0 0>
      }
      tintColor {colorUseRule=UseColor_Specified, specifiedColor=<rgba 1 1 1 1>}
      uVRegion {bIsValid=false, max=<0 0>, min=<0 0>}
     }
     selectedBackgroundColor {colorUseRule=UseColor_Foreground, specifiedColor=<rgba 1 0 1 1>}
     shadowColorAndOpacity=<rgba 0 0 0 1>
     shadowOffset=<0 0>
     strikeBrush {
      bIsDynamicallyLoaded=false
      drawAs=Image
      imageType=NoImage
      mirroring=NoMirror
      resourceName=None
      resourceObject=null
      tiling=NoTile
      imageSize=<32 32>
      margin=<ltrb 0 0 0 0>
      outlineSettings {
       bUseBrushTransparency=false
       roundingType=HalfHeightRadius
       width=0
       color {colorUseRule=UseColor_Specified, specifiedColor=<rgba 0 0 0 0>}
       cornerRadii=<0 0 0 0>
      }
      tintColor {colorUseRule=UseColor_Specified, specifiedColor=<rgba 1 1 1 1>}
      uVRegion {bIsValid=false, max=<0 0>, min=<0 0>}
     }
     underlineBrush {
      bIsDynamicallyLoaded=false
      drawAs=Image
      imageType=NoImage
      mirroring=NoMirror
      resourceName=None
      resourceObject=null
      tiling=NoTile
      imageSize=<32 32>
      margin=<ltrb 0 0 0 0>
      outlineSettings {
       bUseBrushTransparency=false
       roundingType=HalfHeightRadius
       width=0
       color {colorUseRule=UseColor_Specified, specifiedColor=<rgba 0 0 0 0>}
       cornerRadii=<0 0 0 0>
      }
      tintColor {colorUseRule=UseColor_Specified, specifiedColor=<rgba 1 1 1 1>}
      uVRegion {bIsValid=false, max=<0 0>, min=<0 0>}
     }
    }
    margin=<ltrb 0 5 0 0>
   }
   slot {class=OverlaySlot, properties {horizontalAlignment=HAlign_Center, verticalAlignment=VAlign_Center, padding=<ltrb 6 0 6 0>}}
  }
  InputActionWidget {widgetClassPath=&CommonActionWidget, bInherited=true, widget=null}
 ]
}
