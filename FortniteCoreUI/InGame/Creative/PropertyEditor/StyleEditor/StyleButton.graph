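# CommonButtonBase row for the Creative property editor's style editor, showing an icon, style name, rich text sample and current-selection diamond.
# SetupStyleButton sets the name, sample text, style set and preview visibility, while SetSelected toggles the diamond.
# UpdateDefaultStyle applies a device's override color, size, outline and shadow through BPi_TextStyleOverrides to the preview, or clears overrides.
# Click and hover play color animations, with hover skipped for touch input.
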WidgetBlueprint /FortniteCoreUI/InGame/Creative/PropertyEditor/StyleEditor/StyleButton
animations [
 OnClick {
  binding_count=4
  length_seconds=0.4
  object=&self:OnClick
  sequence {bindings [
    Image_Icon {
     id=#1
     possessed_class=Image
     tracks [{
       type=MovieSceneColorTrack
       display_name=Color and Opacity
       sections [{
         type=MovieSceneColorSection
         end_frame=8
         start_frame=0
         channels [
          Color.R {type=MovieSceneScriptingFloatChannel, keys [[0,1],[4,0],[8,1]]}
          Color.G {type=MovieSceneScriptingFloatChannel, keys [[0,1],[4,0],[8,1]]}
          Color.B {type=MovieSceneScriptingFloatChannel, keys [[0,1],[4,0],[8,1]]}
          Color.A {type=MovieSceneScriptingFloatChannel, const [1,1]}
         ]
        }]
      }]
    }
    Root {
     id=#2
     possessed_class=Border
     tracks [
      {
       type=MovieSceneColorTrack
       display_name=Brush Color
       sections [{
         type=MovieSceneColorSection
         end_frame=8
         start_frame=0
         channels [
          Color.R {type=MovieSceneScriptingFloatChannel, keys [[0,0.0024],[4,1],[8,0.0024]]}
          Color.G {type=MovieSceneScriptingFloatChannel, keys [[0,0.0513],[4,1],[8,0.0513]]}
          Color.B {type=MovieSceneScriptingFloatChannel, keys [[0,0.3864],[4,1],[8,0.3864]]}
          Color.A {type=MovieSceneScriptingFloatChannel, const [1,3]}
         ]
        }]
      }
      {
       type=MovieScene2DTransformTrack
       sections [{
         type=MovieScene2DTransformSection
         end_frame=8
         start_frame=0
         channels [
          Scale.X {type=MovieSceneScriptingFloatChannel, keys [[0,1],[4,1.02],[8,1]]}
          Scale.Y {type=MovieSceneScriptingFloatChannel, const [1,2]}
         ]
        }]
      }
     ]
    }
    Border_Name {
     id=#3
     possessed_class=Border
     tracks [{
       type=MovieSceneColorTrack
       display_name=Content Color and Opacity
       sections [{
         type=MovieSceneColorSection
         end_frame=8
         start_frame=0
         channels [
          Color.R {type=MovieSceneScriptingFloatChannel, keys [[0,0.0331],[4,0.0103],[8,0.0331]]}
          Color.G {type=MovieSceneScriptingFloatChannel, keys [[0,0.7011],[4,0.0103],[8,0.7011]]}
          Color.B {type=MovieSceneScriptingFloatChannel, keys [[0,1],[4,0.0103],[8,1]]}
          Color.A {type=MovieSceneScriptingFloatChannel, const [1,3]}
         ]
        }]
      }]
    }
    Border_Value {
     id=#4
     possessed_class=Border
     tracks [{
       type=MovieSceneColorTrack
       display_name=Brush Color
       sections [{
         type=MovieSceneColorSection
         end_frame=8
         start_frame=0
         channels [
          Color.R {type=MovieSceneScriptingFloatChannel, keys [[0,0.0024],[4,0.0103],[8,0.0024]]}
          Color.G {type=MovieSceneScriptingFloatChannel, keys [[0,0.016],[4,0.0103],[8,0.016]]}
          Color.B {type=MovieSceneScriptingFloatChannel, keys [[0,0.117],[4,0.0103],[8,0.117]]}
          Color.A {type=MovieSceneScriptingFloatChannel, const [1,3]}
         ]
        }]
      }]
    }
   ]}
 }
 OnHover {
  binding_count=3
  length_seconds=0.2
  object=&self:OnHover
  sequence {bindings [
    Root {
     id=#2
     possessed_class=Border
     tracks [{
       type=MovieSceneColorTrack
       display_name=Brush Color
       sections [{
         type=MovieSceneColorSection
         end_frame=4
         start_frame=0
         channels [
          Color.R {type=MovieSceneScriptingFloatChannel, keys [[0,0.0024],[4,0.0044]]}
          Color.G {type=MovieSceneScriptingFloatChannel, keys [[0,0.0513],[4,0.3712]]}
          Color.B {type=MovieSceneScriptingFloatChannel, keys [[0,0.3864],[4,0.8308]]}
          Color.A {type=MovieSceneScriptingFloatChannel, const [1,2]}
         ]
        }]
      }]
    }
    Border_Name {
     id=#3
     possessed_class=Border
     tracks [
      {
       type=MovieSceneColorTrack
       display_name=Content Color and Opacity
       sections [{
         type=MovieSceneColorSection
         end_frame=4
         start_frame=0
         channels [
          Color.R {type=MovieSceneScriptingFloatChannel, keys [[0,0.0331],[4,1]]}
          Color.G {type=MovieSceneScriptingFloatChannel, keys [[0,0.7011],[4,1]]}
          Color.B {type=MovieSceneScriptingFloatChannel, const [1,2]}
          Color.A {type=MovieSceneScriptingFloatChannel, const [1,2]}
         ]
        }]
      }
      {
       type=MovieSceneColorTrack
       display_name=Brush Color
       sections [{
         type=MovieSceneColorSection
         end_frame=0
         start_frame=0
         channels [
          Color.R {type=MovieSceneScriptingFloatChannel, const [1,1]}
          Color.G {type=MovieSceneScriptingFloatChannel, const [1,1]}
          Color.B {type=MovieSceneScriptingFloatChannel, const [1,1]}
          Color.A {type=MovieSceneScriptingFloatChannel, const [1,1]}
         ]
        }]
      }
     ]
    }
    Border_Value {
     id=#5
     possessed_class=Border
     tracks [{
       type=MovieSceneColorTrack
       display_name=Brush Color
       sections [{
         type=MovieSceneColorSection
         end_frame=4
         start_frame=0
         channels [
          Color.R {type=MovieSceneScriptingFloatChannel, const [0.0024,2]}
          Color.G {type=MovieSceneScriptingFloatChannel, keys [[0,0.016],[4,0.0513]]}
          Color.B {type=MovieSceneScriptingFloatChannel, keys [[0,0.117],[4,0.3864]]}
          Color.A {type=MovieSceneScriptingFloatChannel, const [1,1]}
         ]
        }]
      }]
    }
   ]}
 }
]
graphs [
 EventGraph {nodes [
   Event On Hovered()
   n1 = IsUsingTouch(PlayerController=GetOwningPlayer())
   > n2 = Branch(Condition=@n1)
   n3 = Get OnHover() -> (OnHover:Widget Animation Object Reference)
   PlayAnimationForward(execute=@n2.else, InAnimation=@n3)
   Event On Clicked()
   > PlayAnimationForward(InAnimation=Get OnClick() -> (OnClick:Widget Animation Object Reference))
   Event On Unhovered()
   > n4 = Branch(Condition=@n1)
   PlayAnimationReverse(execute=@n4.else, InAnimation=@n3)
   Event PreConstruct()
   > "SetText (Text)"(self=Get Text_StyleName() -> (Text_StyleName:Common Text Object Reference), InText=Get ButtonText() -> (ButtonText:Text))
   > Set Visibility(
    Visibility=Select(Option 0=Collapsed, Option 1=SelfHitTestInvisible, Index=Get ShowRichTextBox() -> (ShowRichTextBox:bool))
    self=Get SizeBox_RichText() -> (SizeBox_RichText:Size Box Object Reference)
   )
   > SetBrush(self=Get Image_Icon() -> (Image_Icon:Image Object Reference), InBrush=Get IconBrush() -> (IconBrush:Slate Brush Structure))
   > Set RenderOpacity!(execute, RenderOpacity:float=0, self:Widget Object Reference=Get Image_currentselected() -> (Image_currentselected:Image Object Reference)) -> (then, Output_Get:float)
  ]}
 SetSampleText {nodes [
   n1 = SetSampleText!() -> (then, SampleText:Text, StyleSet:Data Table Object Reference) {is=title K2Node_FunctionEntry}
   n2 = Get RichText_SampleText() -> (RichText_SampleText:Rich Text Block Object Reference)
   > SetText(self=@n2, InText=@n1)
   > SetTextStyleSet!(execute, self:Rich Text Block Object Reference=@n2, NewTextStyleSet:Data Table Object Reference=@n1.StyleSet) -> (then)
  ]}
 SetSelected {nodes [
   n1 = SetSelected!() -> (then, Selected:bool=~) {is=title K2Node_FunctionEntry}
   > n2 = Branch(Condition=@n1)
   n3 = Get Image_currentselected()
   > Set RenderOpacity(RenderOpacity=1, self=@n3)
   Set RenderOpacity(execute=@n2.else, RenderOpacity=0, self=@n3)
  ]}
 SetStyleName {nodes [
   n1 = SetStyleName!() -> (then, Text:Text) {is=title K2Node_FunctionEntry}
   > "SetText (Text)"(self=Get Text_StyleName(), InText=@n1)
  ]}
 SetupStyleButton {nodes [
   n1 = SetupStyleButton!() -> (then, Label:Text, SampleText:Text, StyleSet:Data Table Object Reference, InShowRichTextBox:bool=~) {is=title K2Node_FunctionEntry}
   > SetStyleName!(execute, self, Text:Text=@n1) -> (then) {is=title K2Node_CallFunction}
   > SetSampleText!(execute, self, SampleText:Text=@n1.SampleText, StyleSet:Data Table Object Reference=@n1.StyleSet) -> (then) {is=title K2Node_CallFunction}
   > n2 = Set ShowRichTextBox(ShowRichTextBox:bool=@n1.InShowRichTextBox) -> (Output_Get:bool)
   > Set Visibility(Visibility=Select(Option 0=Collapsed, Option 1=SelfHitTestInvisible, Index=@n2), self=Get SizeBox_RichText())
  ]}
 UpdateDefaultStyle {nodes [
   n1 = UpdateDefaultStyle!() -> (then, Initiating Device:Object Reference) {is=title K2Node_FunctionEntry}
   > n2 = Cast To BPi_TextStyleOverrides!(execute, Object=@n1) -> (then, CastFailed, AsBPi Text Style Overrides:BPi Text Style Overrides Interface) {is=title K2Node_DynamicCast}
   > n3 = GetDefaultStyleOverride!(execute, self:BPi Text Style Overrides Interface=@n2) -> (then, HasOverride:bool, TextColor:Linear Color Structure, ShadowOffset:double, Size:int=0, OutlineStrength:int=0)
   > n4 = Branch(Condition=@n3)
   > n5 = Set StartingOverride(StartingOverride:Text Block Style Structure=Get DefaultTextStyleOverride!(self:Rich Text Block Object Reference=Get RichText_SampleText()) -> (DefaultTextStyleOverride:Text Block Style Structure)) -> (Output_Get:Text Block Style Structure)
   n6 = Break Text Block Style!(TextBlockStyle:"Text Block Style Structure (by ref)"=@n5) -> (Font:Slate Font Info Structure) {is=title K2Node_BreakStruct}
   > n7 = Set members in Font Outline Settings!(
    execute:Exec
    StructRef:"Font Outline Settings Structure (by ref)"=Break Slate Font Info!(SlateFontInfo:"Slate Font Info Structure (by ref)"=@n6) -> (OutlineSettings:Font Outline Settings Structure, bForceMonospaced:bool=~, bMaterialIsStencil:bool=~, MonospacedWidth:float=~) {is=title K2Node_BreakStruct}
    OutlineSize:int=@n3.OutlineStrength
   ) -> (then:Exec, StructOut:"Font Outline Settings Structure (by ref)") {is=title K2Node_SetFieldsInStruct}
   > n8 = Set members in Slate Font Info!(
    execute:Exec
    StructRef:"Slate Font Info Structure (by ref)"=@n6
    OutlineSettings:Font Outline Settings Structure=@n7
    Size:float="To Float (Integer)"(InInt=@n3.Size)
   ) -> (then:Exec, StructOut:"Slate Font Info Structure (by ref)") {is=title K2Node_SetFieldsInStruct}
   > n9 = Set members in Text Block Style!(
    execute:Exec
    StructRef:"Text Block Style Structure (by ref)"=@n5
    Font:Slate Font Info Structure=@n8
    ColorAndOpacity:Slate Color Structure=Make Slate Color(SpecifiedColor=@n3.TextColor)
    ShadowOffset:Vector 2D Structure=MakeVector2D(X=@n3.ShadowOffset, Y=@n3.ShadowOffset)
   ) -> (then:Exec, StructOut:"Text Block Style Structure (by ref)") {is=title K2Node_SetFieldsInStruct}
   > SetDefaultTextStyle!(execute, self:Rich Text Block Object Reference=Get RichText_SampleText(), InDefaultTextStyle:"Text Block Style Structure (by ref)"=@n9) -> (then)
   ClearAllDefaultStyleOverrides(execute=@n4.else, self:Rich Text Block Object Reference=Get RichText_SampleText())
  ]}
]
variables [
 ButtonText {type=string, default="{Style Name}"}
 IconBrush {
  type=SlateBrush
  default {
   bIsDynamicallyLoaded=false
   drawAs=NoDrawType
   imageType=NoImage
   mirroring=NoMirror
   resourceName=None
   resourceObject=null
   tiling=NoTile
   imageSize=<0 0>
   margin=<ltrb 0 0 0 0>
   outlineSettings {
    bUseBrushTransparency=false
    roundingType=HalfHeightRadius
    width=0
    color {colorUseRule=UseColor_Specified, specifiedColor=<rgba 0 0 0 0>}
    cornerRadii=<0 0 0 1>
   }
   tintColor {colorUseRule=UseColor_Specified, specifiedColor=<rgba 1 1 1 1>}
   uVRegion {bIsValid=false, max=<0 0>, min=<0 0>}
  }
 }
 ShowRichTextBox {type=boolean, default=true}
]
widgets {
 info {inheritedWidgetCount=1, parentClass=&CommonButtonBase, rootWidgetClass=&CanvasPanel, widgetCount=19}
 widgets [
  CanvasPanel_2 {widgetClassPath=&CanvasPanel, widget=CanvasPanel_2}
  Root {
   widgetClassPath=&Border
   parent=CanvasPanel_2
   widget=Root
   properties {brushColor=<rgba 0.0024 0.0513 0.3864 1>, padding=<ltrb 0 0 0 0>, renderTransformPivot=<0 0>}
   slot {class=CanvasPanelSlot, properties {bAutoSize=true, layoutData=~}}
  }
  HorizontalBox_0 {widgetClassPath=&HorizontalBox, parent=Root, widget=HorizontalBox_0, slot {class=BorderSlot, properties {padding=<ltrb 0 0 0 0>}}}
  Current {
   widgetClassPath=&Overlay
   parent=HorizontalBox_0
   widget=Current
   slot {class=HorizontalBoxSlot, properties {horizontalAlignment=HAlign_Center, verticalAlignment=VAlign_Center, padding=<ltrb 16 0 0 0>}}
  }
  Image_currentselected {
   widgetClassPath=&Image
   bIsVariable=true
   parent=Current
   widget=Image_currentselected
   properties {brush {
     bIsDynamicallyLoaded=false
     drawAs=Image
     imageType=NoImage
     mirroring=NoMirror
     resourceName=None
     resourceObject=/Game/Athena/HUD/Art/T_UI_Indicator_Diamond
     tiling=NoTile
     imageSize=<10 10>
     margin=<ltrb 0 0 0 0>
     outlineSettings {
      bUseBrushTransparency=false
      roundingType=HalfHeightRadius
      width=0
      color {colorUseRule=UseColor_Specified, specifiedColor=<rgba 0 0 0 0>}
      cornerRadii=<0 0 0 0>
     }
     tintColor {colorUseRule=UseColor_Specified, specifiedColor=<rgba 0.0331 0.7011 1 1>}
     uVRegion {bIsValid=false, max=<0 0>, min=<0 0>}
    }}
   slot {class=OverlaySlot}
  }
  Image_Icon {
   widgetClassPath=&Image
   bIsVariable=true
   parent=HorizontalBox_0
   widget=Image_Icon
   properties {brush {
     bIsDynamicallyLoaded=false
     drawAs=NoDrawType
     imageType=NoImage
     mirroring=NoMirror
     resourceName=None
     resourceObject=null
     tiling=NoTile
     imageSize=<32 32>
     margin=<ltrb 0 0 0 0>
     outlineSettings {
      bUseBrushTransparency=false
      roundingType=HalfHeightRadius
      width=0
      color {colorUseRule=UseColor_Specified, specifiedColor=<rgba 0 0 0 0>}
      cornerRadii=<0 0 0 0>
     }
     tintColor {colorUseRule=UseColor_Specified, specifiedColor=<rgba 1 1 1 1>}
     uVRegion {bIsValid=false, max=<0 0>, min=<0 0>}
    }}
   slot {class=HorizontalBoxSlot, properties {horizontalAlignment=HAlign_Center, verticalAlignment=VAlign_Center}}
  }
  FunctionAndNames {
   widgetClassPath=&SizeBox
   parent=HorizontalBox_0
   widget=FunctionAndNames
   properties {bOverride_WidthOverride=true, widthOverride=300}
   slot {class=HorizontalBoxSlot, properties {verticalAlignment=VAlign_Center, padding=<ltrb 20 0 0 0>, size {sizeRule=Fill, value=1}}}
  }
  Name {
   widgetClassPath=&GridPanel
   parent=FunctionAndNames
   widget=Name
   properties {columnFill [1], rowFill [1]}
   slot {class=SizeBoxSlot, properties {verticalAlignment=VAlign_Center}}
  }
  Spacer_Style {
   widgetClassPath=&Spacer
   parent=Name
   widget=Spacer_Style
   properties {size=<300 56>}
   slot {class=GridSlot, properties {horizontalAlignment=HAlign_Left, verticalAlignment=VAlign_Top}}
  }
  Border_Name {
   widgetClassPath=&Border
   parent=Name
   widget=Border_Name
   properties {
    verticalAlignment=VAlign_Center
    background {
     bIsDynamicallyLoaded=false
     drawAs=NoDrawType
     imageType=NoImage
     mirroring=NoMirror
     resourceName=None
     resourceObject=null
     tiling=NoTile
     imageSize=<32 32>
     margin=<ltrb 0 0 0 0>
     outlineSettings {
      bUseBrushTransparency=false
      roundingType=HalfHeightRadius
      width=0
      color {colorUseRule=UseColor_Specified, specifiedColor=<rgba 0 0 0 0>}
      cornerRadii=<0 0 0 0>
     }
     tintColor {colorUseRule=UseColor_Specified, specifiedColor=<rgba 1 1 1 1>}
     uVRegion {bIsValid=false, max=<0 0>, min=<0 0>}
    }
    contentColorAndOpacity=<rgba 0.0331 0.7011 1 1>
    padding=<ltrb 0 4 0 0>
   }
   slot {class=GridSlot, properties {verticalAlignment=VAlign_Center}}
  }
  Text_StyleName {
   widgetClassPath=&CommonTextBlock
   bIsVariable=true
   parent=Border_Name
   widget=Text_StyleName
   properties {
    lineHeightPercentage=1.29
    scrollStyle=&TextScrollStyle-Base_C
    style=/FortUILibrary/Styles/Typography/_TextStyle_Body1._TextStyle_Body1_C
    text="{Style Name}"
    textTransformPolicy=ToUpper
    font {
     bForceMonospaced=false
     bMaterialIsStencil=false
     fontMaterial=null
     fontObject=&BurbankSmall
     letterSpacing=0
     monospacedWidth=1
     size=18
     skewAmount=0
     typefaceFontName=Bold
     outlineSettings {
      bApplyOutlineToDropShadows=false
      bMiteredCorners=false
      bSeparateFillAlpha=false
      outlineMaterial=null
      outlineSize=0
      outlineColor=<rgba 0 0 0 1>
     }
    }
    shadowOffset=<0 0>
    strikeBrush {
     bIsDynamicallyLoaded=false
     drawAs=NoDrawType
     imageType=NoImage
     mirroring=NoMirror
     resourceName=None
     resourceObject=null
     tiling=NoTile
     imageSize=<32 32>
     margin=<ltrb 0 0 0 0>
     outlineSettings {
      bUseBrushTransparency=false
      roundingType=HalfHeightRadius
      width=0
      color {colorUseRule=UseColor_Specified, specifiedColor=<rgba 0 0 0 0>}
      cornerRadii=<0 0 0 0>
     }
     tintColor {colorUseRule=UseColor_Specified, specifiedColor=<rgba 1 1 1 1>}
     uVRegion {bIsValid=false, max=<0 0>, min=<0 0>}
    }
   }
   slot {class=BorderSlot, properties {verticalAlignment=VAlign_Center, padding=<ltrb 0 4 0 0>}}
  }
  SizeBox_RichText {
   widgetClassPath=&SizeBox
   bIsVariable=true
   parent=HorizontalBox_0
   widget=SizeBox_RichText
   properties {bOverride_WidthOverride=true, widthOverride=370}
   slot {class=HorizontalBoxSlot, properties {padding=<ltrb 0 0 12 0>, size {sizeRule=Fill, value=1}}}
  }
  Value {
   widgetClassPath=&GridPanel
   parent=SizeBox_RichText
   widget=Value
   properties {columnFill [1], rowFill [1]}
   slot {class=SizeBoxSlot}
  }
  Spacer_Value {
   widgetClassPath=&Spacer
   parent=Value
   widget=Spacer_Value
   properties {size=<360 0>}
   slot {class=GridSlot, properties {horizontalAlignment=HAlign_Left, verticalAlignment=VAlign_Top}}
  }
  Border_Value {
   widgetClassPath=&Border
   parent=Value
   widget=Border_Value
   properties {horizontalAlignment=HAlign_Center, verticalAlignment=VAlign_Center, brushColor=<rgba 0.0024 0.016 0.117 1>, padding=<ltrb 0 0 0 0>}
   slot {class=GridSlot, properties {padding=<ltrb 0 4 0 4>}}
  }
  SizeBox_2 {
   widgetClassPath=&SizeBox
   parent=Border_Value
   widget=SizeBox_2
   properties {bOverride_WidthOverride=true, clipping=ClipToBounds, widthOverride=300}
   slot {class=BorderSlot, properties {horizontalAlignment=HAlign_Center, verticalAlignment=VAlign_Center, padding=<ltrb 0 0 0 0>}}
  }
  RichText_SampleText {
   widgetClassPath=/Script/UMG.RichTextBlock
   bIsVariable=true
   parent=SizeBox_2
   widget=RichText_SampleText
   properties {
    justification=Center
    text="<EMPHASIS>Sample Text</>"
    textStyleSet=/Game/Athena/UI/InGame/Creative/PropertyEditor/StyleEditor/RichTextStyleSets/CreativeDefaultRichTextStyles
    defaultTextStyleOverride {
     overflowPolicy=Clip
     transformPolicy=None
     colorAndOpacity {colorUseRule=UseColor_Specified, specifiedColor=<rgba 1 0 0.8057 1>}
     font {
      bForceMonospaced=false
      bMaterialIsStencil=false
      fontMaterial=null
      fontObject=&BurbankSmall
      letterSpacing=0
      monospacedWidth=1
      size=18
      skewAmount=0
      typefaceFontName=Black
      outlineSettings {
       bApplyOutlineToDropShadows=false
       bMiteredCorners=false
       bSeparateFillAlpha=false
       outlineMaterial=null
       outlineSize=0
       outlineColor=<rgba 0 0 0 1>
      }
     }
     highlightColor {colorUseRule=UseColor_Specified, specifiedColor=<rgba 0 0 0 1>}
     highlightShape {
      bIsDynamicallyLoaded=false
      drawAs=Image
      imageType=NoImage
      mirroring=NoMirror
      resourceName=None
      resourceObject=null
      tiling=NoTile
      imageSize=<32 32>
      margin=<ltrb 0 0 0 0>
      outlineSettings {
       bUseBrushTransparency=false
       roundingType=HalfHeightRadius
       width=0
       color {colorUseRule=UseColor_Specified, specifiedColor=<rgba 0 0 0 0>}
       cornerRadii=<0 0 0 0>
      }
      tintColor {colorUseRule=UseColor_Specified, specifiedColor=<rgba 1 1 1 1>}
      uVRegion {bIsValid=false, max=<0 0>, min=<0 0>}
     }
     selectedBackgroundColor {colorUseRule=UseColor_Foreground, specifiedColor=<rgba 1 0 1 1>}
     shadowColorAndOpacity=<rgba 0 0 0 1>
     shadowOffset=<0 0>
     strikeBrush {
      bIsDynamicallyLoaded=false
      drawAs=Image
      imageType=NoImage
      mirroring=NoMirror
      resourceName=None
      resourceObject=null
      tiling=NoTile
      imageSize=<32 32>
      margin=<ltrb 0 0 0 0>
      outlineSettings {
       bUseBrushTransparency=false
       roundingType=HalfHeightRadius
       width=0
       color {colorUseRule=UseColor_Specified, specifiedColor=<rgba 0 0 0 0>}
       cornerRadii=<0 0 0 0>
      }
      tintColor {colorUseRule=UseColor_Specified, specifiedColor=<rgba 1 1 1 1>}
      uVRegion {bIsValid=false, max=<0 0>, min=<0 0>}
     }
     underlineBrush {
      bIsDynamicallyLoaded=false
      drawAs=Image
      imageType=NoImage
      mirroring=NoMirror
      resourceName=None
      resourceObject=null
      tiling=NoTile
      imageSize=<32 32>
      margin=<ltrb 0 0 0 0>
      outlineSettings {
       bUseBrushTransparency=false
       roundingType=HalfHeightRadius
       width=0
       color {colorUseRule=UseColor_Specified, specifiedColor=<rgba 0 0 0 0>}
       cornerRadii=<0 0 0 0>
      }
      tintColor {colorUseRule=UseColor_Specified, specifiedColor=<rgba 1 1 1 1>}
      uVRegion {bIsValid=false, max=<0 0>, min=<0 0>}
     }
    }
   }
   slot {class=SizeBoxSlot}
  }
  Border_0 {
   widgetClassPath=&Border
   parent=CanvasPanel_2
   widget=Border_0
   properties {
    horizontalAlignment=HAlign_Center
    verticalAlignment=VAlign_Center
    background {
     bIsDynamicallyLoaded=false
     drawAs=Border
     imageType=NoImage
     mirroring=NoMirror
     resourceName=None
     resourceObject=null
     tiling=NoTile
     imageSize=<32 32>
     margin=<ltrb 0 0 0 0>
     outlineSettings {
      bUseBrushTransparency=false
      roundingType=HalfHeightRadius
      width=0
      color {colorUseRule=UseColor_Specified, specifiedColor=<rgba 0 0 0 0>}
      cornerRadii=<0 0 0 0>
     }
     tintColor {colorUseRule=UseColor_Specified, specifiedColor=<rgba 1 1 1 1>}
     uVRegion {bIsValid=false, max=<0 0>, min=<0 0>}
    }
   }
   slot {
    class=CanvasPanelSlot
    properties {bAutoSize=true, layoutData {alignment=<0 0.5>, anchors {maximum=<0 0.5>, minimum=<0 0.5>}, offsets=<ltrb 0 0 100 30>}}
   }
  }
  InputActionWidget {
   widgetClassPath=&CommonActionWidget
   bInherited=true
   bIsVariable=true
   parent=Border_0
   widget=InputActionWidget
   properties {
    progressMaterialParam=Percentage
    progressMaterialBrush {
     bIsDynamicallyLoaded=false
     drawAs=Image
     imageType=NoImage
     mirroring=NoMirror
     resourceName=None
     resourceObject=/Game/UI/Materials/HollowCircularProgressBar
     tiling=NoTile
     imageSize=<36 36>
     margin=<ltrb 0 0 0 0>
     outlineSettings {
      bUseBrushTransparency=false
      roundingType=HalfHeightRadius
      width=0
      color {colorUseRule=UseColor_Specified, specifiedColor=<rgba 0 0 0 0>}
      cornerRadii=<0 0 0 0>
     }
     tintColor {colorUseRule=UseColor_Specified, specifiedColor=<rgba 1 1 0 1>}
     uVRegion {bIsValid=false, max=<0 0>, min=<0 0>}
    }
   }
   slot {class=BorderSlot, properties {horizontalAlignment=HAlign_Center, verticalAlignment=VAlign_Center}}
  }
 ]
}
